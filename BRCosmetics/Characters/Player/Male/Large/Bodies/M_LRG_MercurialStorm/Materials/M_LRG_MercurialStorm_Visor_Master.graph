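# Master character material for the MercurialStorm large male visor, with an optional animated dissolve.
# Character PreFX feeds a Fresnel-masked visor function, metals LUT and clearcoat, then PostFX and cinematic PostFX.
# A static switch selects the dissolve, where sphere masks sized by remapped progress or direct radius form a band.
# Two panning noise samples drive its colored emissive edge, and a channel mask limits the blend into base color, emissive and opacity mask.
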Material /BRCosmetics/Characters/Player/Male/Large/Bodies/M_LRG_MercurialStorm/Materials/M_LRG_MercurialStorm_Visor_Master
graph {
 properties {
  bUseMaterialAttributes=true
  bUsedWithMeshDeformer=true
  bUsedWithMorphTargets=true
  bUsedWithSkeletalMesh=true
  pixelDepthOffsetMode=PDOM_Legacy
  usedShadingModels=Default Lit
 }
 nodes [
  n1 = VectorParameter() {group=VisorDisolve, parameterName=MaskLocation, channelNames {a="", b=Z, g=Y, r=X}, defaultValue=<rgba 0 0 0 1>}
  Add(A=TransformPosition(None=ActorPositionWS()) {transformSourceType=TRANSFORMPOSSOURCE_World}, B=@n1)
  Constant() {r=150}
  LinearInterpolate(
   A=Constant3Vector() {constant=<rgba 1 0 0 0>}
   B=Constant3Vector() {constant=<rgba 0 0 1 0>}
   Alpha=ScalarParameter() {group=VisorDisolve, parameterName=Visor_Dissolve_Progress, sliderMax=1}
  )
  n2 = MaterialFunctionCall() {materialFunction=&BaseFN_CharacterFunction_PreFX}
  n3 = Constant() {r=1}
  n4 = OneMinus(None=@n2.Fresnel)
  n5 = ScalarParameter() {defaultValue=2, parameterName=Visor_Fresnel}
  n6 = MaterialFunctionCall(
   Mask=NamedRerouteUsage()
   Materials=MaterialFunctionCall!(Mask=Power(Base=@n4, Exp=@n5), Material=@n2.Result) {materialFunction=/Game/Characters/Player/Female/Medium/Bodies/F_MED_Binary/Materials/MF_Visor}.Result
  ) {materialFunction=&MF_MetalsLUT}
  n7 = StaticSwitchParameter(
   True=NamedRerouteUsage()
   False=NamedRerouteDeclaration(None=MaterialFunctionCall!(Alpha=@n3, In=@n6.Result) {materialFunction=/Game/Weapons/WeaponSkins/MaterialFunction/Clearcoat}.Result) {name=AfterClearCoat_MA, nodeColor=<rgba 0.1953 1 0 1>}
  ) {group=VisorDisolve, parameterName=Use_Visor_Dissolve}
  n8 = MaterialFunctionCall(
   AO=@n2.AO
   Distance Blend=@n2.AO
   Fresnel=@n2.AO
   Gmap=@n2.AO
   Pre Skinned Local Position=NamedRerouteDeclaration(None=@n2.Pre skinned Local Position) {name=PreSkinned Position, nodeColor=<rgba 0 1 0.4609 1>}
   Material=@n7
  ) {materialFunction=&BaseFN_CharacterFunction_PostFX}
  MaterialFunctionCall(In=@n8.Result) {materialFunction=&CINE_MF_CharPostFX}
  NamedRerouteDeclaration(None=TextureSampleParameter2D() {parameterName=SkinFx_Mask}.RGBA) {name=SkinFX_Mask, nodeColor=<rgba 1 0 0.8437 1>}
  n9 = TextureSampleParameter2D(UVs=Panner() {speedY=0.075}) {
   group=VisorDisolve
   parameterName=NoiseMask
   samplerType=SAMPLERTYPE_Masks
   texture=/BRCosmetics/Characters/Player/Female/Medium/Bodies/F_MED_MeteorWoman_Suit/Textures/T_F_MED_MeteorWoman_Noise
  }
  n10 = TextureSampleParameter2D(UVs=Multiply(A=Panner() {speedY=0.025}) {constB=5}) {
   group=VisorDisolve
   parameterName=NoiseMask
   samplerType=SAMPLERTYPE_Masks
   texture=&T_F_MED_MeteorWoman_Noise
  }
  n11 = Multiply(
   A=Multiply(
    A=StaticComponentMaskParameter(None=@n9.RGBA) {defaultG=true, parameterName=NoiseMask_Channel}
    B=StaticComponentMaskParameter(None=@n10.RGBA) {defaultG=true, parameterName=NoiseMask_Channel}
   )
   B=NamedRerouteUsage()
  )
  n12 = Multiply(A=@n11, B=VectorParameter() {group=VisorDisolve, parameterName=VisorDissolve_Color, defaultValue=<rgba 0.0571 0.7228 0.9531 1>}.RGB)
  n13 = ScalarParameter() {defaultValue=20, group=VisorDisolve, parameterName=VisorDissolve_Emissive_Power}
  n14 = NamedRerouteUsage()
  NamedRerouteDeclaration(None=SetMaterialAttributes!(MaterialAttributes, Base Color=@n12, Emissive Color=Multiply(A=@n12, B=@n13) {constB=20}, Opacity Mask=@n14) {attributeSetTypes [#1, #2, #3]}) {name=VisorDissolve, nodeColor=<rgba 0.9453 1 0 1>}
  n15 = GetMaterialAttributes(None=NamedRerouteUsage()) {attributeGetTypes [#1, #2]}
  n16 = GetMaterialAttributes(None=NamedRerouteUsage()) {attributeGetTypes [#1, #2, #3]}
  n17 = StaticComponentMaskParameter(None=NamedRerouteUsage()) {defaultG=true, group=VisorDisolve, parameterName=VisorDissolve_MaskChannel}
  n18 = Multiply(A=NamedRerouteUsage(), B=@n17)
  n19 = OneMinus(None=@n17)
  NamedRerouteDeclaration(None=SetMaterialAttributes(
    MaterialAttributes=NamedRerouteUsage()
    Base Color=LinearInterpolate(A=@n15.Base Color, B=@n16.Base Color, Alpha=@n18)
    Emissive Color=LinearInterpolate(A=@n15.Emissive Color, B=@n16.Emissive Color, Alpha=@n18) {constAlpha=0}
    Opacity Mask=LinearInterpolate(A=@n16.Opacity Mask, Alpha=@n19)
   ) {attributeSetTypes [#1, #2, #3]}) {name=VisorDissolve_AndOthers_MA, nodeColor=<rgba 1 0 0.9141 1>}
  n20 = NamedRerouteUsage()
  n21 = SphereMask(A=@n20, B=@n1, Radius=NamedRerouteUsage()) {attenuationRadius=45}
  NamedRerouteDeclaration(None=@n21) {name=SphereMask, nodeColor=<rgba 1 0 0.0469 1>}
  n22 = Power(
   Base=SphereMask(A=@n20, B=@n1, Radius=NamedRerouteUsage(), Hardness=ScalarParameter() {defaultValue=100, group=VisorDisolve, parameterName=MaskHardness}) {attenuationRadius=150}
   Exp=ScalarParameter() {defaultValue=1, group=VisorDisolve, parameterName=InnerBandSoftness}
  )
  NamedRerouteDeclaration(None=Saturate(None=Subtract(A=@n21, B=@n22))) {name=SphereMaskBand, nodeColor=<rgba 1 0.6563 0 1>}
  n23 = ScalarParameter() {group=VisorDisolve, parameterName=Visor_Dissolve_Progress, sliderMax=1}
  n24 = StaticSwitchParameter(
   True=ScalarParameter() {group=VisorDisolve, parameterName=MaskRadius}
   False=MaterialFunctionCall(
    Input=@n23
    Input Low=Constant()
    Input High=Constant() {r=1}
    Target Low=ScalarParameter() {group=VisorDisolve, parameterName=Min_Radius_Range}
    Target High=ScalarParameter() {group=VisorDisolve, parameterName=Max_Radius_Range}
   ) {materialFunction=&RemapValueRange}.Result
  ) {group=VisorDisolve, parameterName=Direct Mask Radius Control}
  NamedRerouteDeclaration(None=@n24) {name=Mask Radius, nodeColor=<rgba 0 1 0.3203 1>}
  NamedRerouteDeclaration(None=Add(A=@n24, B=ScalarParameter() {defaultValue=1, group=VisorDisolve, parameterName=MaskBandThickness})) {name=Mask Radius 2, nodeColor=<rgba 1 0 0.3984 1>}
  NamedRerouteUsage()
  NamedRerouteUsage()
  NamedRerouteUsage()
  OneMinus(None=@n23)
  n25 = Constant()
  SetMaterialAttributes!(
   MaterialAttributes=@n7
   Base Color=@n25
   Emissive Color=NamedRerouteUsage()
   Opacity Mask=Constant() {r=1}
   Metallic=@n25
   Specular=@n25
   World Position Offset=Multiply(A=VertexNormalWS()) {constB=3}
  ) {attributeSetTypes [
    #1
    #2
    #3
    #4
    #5
    #6
   ]}
  SetMaterialAttributes!(MaterialAttributes, Base Color=NamedRerouteUsage()) {attributeSetTypes [#1]}
  n26 = Multiply(A=Time(), B=ScalarParameter() {defaultValue=0.1, group=VisorDisolve, parameterName=Debug_Animation_Speed}) {constB=0.05}
  StaticSwitchParameter(True=Abs(None=Sine(None=@n26)), False=@n23) {group=VisorDisolve, parameterName=Debug Animation}
  Subtract(A=@n14, B=@n11)
  TransformPosition(None=@n20) {transformType=TRANSFORMPOSSOURCE_World}
 ]
}
