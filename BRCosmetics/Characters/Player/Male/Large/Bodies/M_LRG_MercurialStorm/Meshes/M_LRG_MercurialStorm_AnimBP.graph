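# Post-process style AnimBP for the MercurialStorm large male body cosmetic, adding procedural corrective motion to armor and volume bones.
# It copies the attached parent's pose, then chains weighted constraints and additive bone offsets for shoulder plates, arms, thighs, grenades and calves.
# Float variables for limb angles, weapon offsets and skydiving drive the corrective alphas.
# A cached pose branches on bIsTargeting to nudge the right shoulder plate before the limb correctives.
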AnimBlueprint /BRCosmetics/Characters/Player/Male/Large/Bodies/M_LRG_MercurialStorm/Meshes/M_LRG_MercurialStorm_AnimBP
target_skeleton=&/M_LRG_MercurialStorm_Skeleton
graphs [
 AnimGraph {nodes [
   Constraint!(ComponentPose, Alpha=1) -> (Pose) {alphaBoolBlend {blendOption=Linear}, boneToModify=~}
   n1 = "Transform (Modify) Bone - Bone: dyn_shoulderPlate_r"(
    ComponentPose=Local To Component(LocalPose=Use cached pose 'TargetingCache'())
    Alpha=Get offset_weapon_0() -> (offset_weapon_0:double)
    Translation="-3,1.5,3"
    Rotation="0,0,0"
    Scale="1,1,1"
   ) {
    rotationMode=BMM_Additive
    rotationSpace=BCS_BoneSpace
    scaleMode=BMM_Additive
    scaleSpace=BCS_BoneSpace
    translationMode=BMM_Additive
    translationSpace=BCS_BoneSpace
    alphaBoolBlend {blendOption=Linear}
    boneToModify {boneName=dyn_shoulderPlate_r}
    translation=<-3 1.5 3>
   }
   n2 = Component To Local(ComponentPose=@n1)
   n3 = Use cached pose 'TargetingCache'()
   n4 = Get bIsTargeting()
   n5 = "Constraint - lowerarm_l_front (RR)"!(
    ComponentPose=Local To Component(LocalPose=Blend Poses by bool(BlendPose_0=@n2, BlendPose_1=@n3, bActiveValue=@n4) {blendType=Linear})
    Alpha=1
    ConstraintWeights_0=0.5
    ConstraintWeights_1:float=0.5
   ) -> (Pose) {
    alphaBoolBlend {blendOption=Linear}
    boneToModify {boneName=lowerarm_l_front}
    constraintSetup [{transformType=Rotation, targetBone {boneName=lowerarm_l}}, {transformType=Rotation, targetBone {boneName=upperarm_l}}]
    constraintWeights [1,1]
   }
   n6 = "Constraint - lowerarm_r_front (RR)"!(ComponentPose=@n5, Alpha=1, ConstraintWeights_0=0.5, ConstraintWeights_1:float=0.5) -> (Pose) {
    alphaBoolBlend {blendOption=Linear}
    boneToModify {boneName=lowerarm_r_front}
    constraintSetup [{transformType=Rotation, targetBone {boneName=lowerarm_r}}, {transformType=Rotation, targetBone {boneName=upperarm_r}}]
    constraintWeights [1,1]
   }
   n7 = "Constraint - lowerarm_l_back (RR)"!(ComponentPose=@n6, Alpha=1, ConstraintWeights_0=0.2, ConstraintWeights_1:float=0.8) -> (Pose) {
    alphaBoolBlend {blendOption=Linear}
    boneToModify {boneName=lowerarm_l_back}
    constraintSetup [{transformType=Rotation, targetBone {boneName=lowerarm_l}}, {transformType=Rotation, targetBone {boneName=upperarm_l}}]
    constraintWeights [1,1]
   }
   n8 = "Constraint - lowerarm_r_back (RR)"!(ComponentPose=@n7, Alpha=1, ConstraintWeights_0=0.5, ConstraintWeights_1:float=0.5) -> (Pose) {
    alphaBoolBlend {blendOption=Linear}
    boneToModify {boneName=lowerarm_r_back}
    constraintSetup [{transformType=Rotation, targetBone {boneName=lowerarm_r}}, {transformType=Rotation, targetBone {boneName=upperarm_r}}]
    constraintWeights [1,1]
   }
   n9 = Get lowerarm_l_front() -> (lowerarm_l_front:double)
   n10 = "Transform (Modify) Bone - Bone: lowerarm_l_front"(ComponentPose=@n8, Alpha=@n9, Translation="5,5,0", Rotation="0,0,0", Scale="1,1,1") {
    rotationMode=BMM_Additive
    rotationSpace=BCS_BoneSpace
    scaleMode=BMM_Additive
    scaleSpace=BCS_BoneSpace
    translationMode=BMM_Additive
    translationSpace=BCS_BoneSpace
    alphaBoolBlend {blendOption=Linear}
    boneToModify {boneName=lowerarm_l_front}
    translation=<5 5>
   }
   n11 = Get lowerarm_r_front() -> (lowerarm_r_front:double)
   n12 = "Transform (Modify) Bone - Bone: lowerarm_r_front"(ComponentPose=@n10, Alpha=@n11, Translation="-5,-5,0", Rotation="0,0,0", Scale="1,1,1") {
    rotationMode=BMM_Additive
    rotationSpace=BCS_BoneSpace
    scaleMode=BMM_Additive
    scaleSpace=BCS_BoneSpace
    translationMode=BMM_Additive
    translationSpace=BCS_BoneSpace
    alphaBoolBlend {blendOption=Linear}
    boneToModify {boneName=lowerarm_r_front}
    translation=<-5 -5>
   }
   n13 = "Transform (Modify) Bone - Bone: lowerarm_l_back"(ComponentPose=@n12, Alpha=@n9, Translation="-5,0,0", Rotation="0,0,0", Scale="1,1,1") {
    rotationMode=BMM_Additive
    rotationSpace=BCS_BoneSpace
    scaleMode=BMM_Additive
    scaleSpace=BCS_BoneSpace
    translationMode=BMM_Additive
    translationSpace=BCS_BoneSpace
    alphaBoolBlend {blendOption=Linear}
    boneToModify {boneName=lowerarm_l_back}
    translation {x=-5}
   }
   n14 = "Transform (Modify) Bone - Bone: lowerarm_r_back"(ComponentPose=@n13, Alpha=@n11, Translation="0,10,0", Rotation="0,0,0", Scale="1,1,1") {
    rotationMode=BMM_Additive
    rotationSpace=BCS_BoneSpace
    scaleMode=BMM_Additive
    scaleSpace=BCS_BoneSpace
    translationMode=BMM_Additive
    translationSpace=BCS_BoneSpace
    alphaBoolBlend {blendOption=Linear}
    boneToModify {boneName=lowerarm_r_back}
    translation {y=10}
   }
   n15 = "Constraint - thigh_volume_bk_l (PP)"!(ComponentPose=@n14, Alpha=1, ConstraintWeights_0=1, ConstraintWeights_1:float=0) -> (Pose) {
    alphaBoolBlend {blendOption=Linear}
    boneToModify {boneName=thigh_volume_bk_l}
    constraintSetup [{targetBone {boneName=thigh_twist_01_l}}, {targetBone {boneName=pelvis}}]
    constraintWeights [1,1]
   }
   n16 = "Constraint - thigh_volume_fn_l (PP)"!(ComponentPose=@n15, Alpha=Get thigh_shape_fn_l() -> (thigh_shape_fn_l:double), ConstraintWeights_0=0.2, ConstraintWeights_1:float=0.8) -> (Pose) {
    alphaBoolBlend {blendOption=Linear}
    boneToModify {boneName=thigh_volume_fn_l}
    constraintSetup [{targetBone {boneName=thigh_twist_01_l}}, {targetBone {boneName=pelvis}}]
    constraintWeights [1,1]
   }
   n17 = "Constraint - thigh_volume_lt_l (PP)"!(ComponentPose=@n16, Alpha=1, ConstraintWeights_0=0.7, ConstraintWeights_1:float=0.3) -> (Pose) {
    alphaBoolBlend {blendOption=Linear}
    boneToModify {boneName=thigh_volume_lt_l}
    constraintSetup [{targetBone {boneName=thigh_twist_01_l}}, {targetBone {boneName=pelvis}}]
    constraintWeights [1,1]
   }
   n18 = "Constraint - thigh_volume_bk_r (PP)"!(ComponentPose=@n17, Alpha=1, ConstraintWeights_0=0, ConstraintWeights_1:float=1) -> (Pose) {
    alphaBoolBlend {blendOption=Linear}
    boneToModify {boneName=thigh_volume_bk_r}
    constraintSetup [{targetBone {boneName=thigh_twist_01_r}}, {targetBone {boneName=pelvis}}]
    constraintWeights [1,1]
   }
   n19 = "Constraint - thigh_volume_fn_r (PP)"!(ComponentPose=@n18, Alpha=Get thigh_shape_fn_r() -> (thigh_shape_fn_r:double), ConstraintWeights_0=0.2, ConstraintWeights_1:float=0.8) -> (Pose) {
    alphaBoolBlend {blendOption=Linear}
    boneToModify {boneName=thigh_volume_fn_r}
    constraintSetup [{targetBone {boneName=thigh_twist_01_r}}, {targetBone {boneName=pelvis}}]
    constraintWeights [1,1]
   }
   n20 = "Constraint - thigh_volume_rt_r (PP)"!(ComponentPose=@n19, Alpha=1, ConstraintWeights_0=0.7, ConstraintWeights_1:float=0.3) -> (Pose) {
    alphaBoolBlend {blendOption=Linear}
    boneToModify {boneName=thigh_volume_rt_r}
    constraintSetup [{targetBone {boneName=thigh_twist_01_r}}, {targetBone {boneName=pelvis}}]
    constraintWeights [1,1]
   }
   n21 = "Transform (Modify) Bone - Bone: thigh_volume_fn_l"(ComponentPose=@n20, Alpha=Get thigh_shape_fn_l(), Translation="15,-15,-5", Rotation="0,-20,0", Scale="1,1,1") {
    rotationMode=BMM_Additive
    rotationSpace=BCS_BoneSpace
    scaleMode=BMM_Additive
    scaleSpace=BCS_BoneSpace
    translationMode=BMM_Additive
    translationSpace=BCS_BoneSpace
    alphaBoolBlend {blendOption=Linear}
    boneToModify {boneName=thigh_volume_fn_l}
    rotation {yaw=-20}
    translation=<15 -15 -5>
   }
   n22 = "Transform (Modify) Bone - Bone: thigh_volume_fn_r"(ComponentPose=@n21, Alpha=Get thigh_shape_fn_r(), Translation="-15,-15,5", Rotation="0,-20,0", Scale="1,1,1") {
    rotationMode=BMM_Additive
    rotationSpace=BCS_BoneSpace
    scaleMode=BMM_Additive
    scaleSpace=BCS_BoneSpace
    translationMode=BMM_Additive
    translationSpace=BCS_BoneSpace
    alphaBoolBlend {blendOption=Linear}
    boneToModify {boneName=thigh_volume_fn_r}
    rotation {yaw=-20}
    translation=<-15 -15 5>
   }
   n23 = "Transform (Modify) Bone - Bone: thigh_volume_bk_l"(ComponentPose=@n22, Alpha=Get thigh_shape_bk_l() -> (thigh_shape_bk_l:double), Translation="12,-5,0", Rotation="0,0,0", Scale="1.5,1.5,1.5") {
    rotationMode=BMM_Additive
    rotationSpace=BCS_BoneSpace
    scaleMode=BMM_Additive
    scaleSpace=BCS_BoneSpace
    translationMode=BMM_Additive
    translationSpace=BCS_BoneSpace
    alphaBoolBlend {blendOption=Linear}
    boneToModify {boneName=thigh_volume_bk_l}
    scale=<1.5 1.5 1.5>
    translation=<12 -5>
   }
   n24 = "Transform (Modify) Bone - Bone: thigh_volume_bk_r"(ComponentPose=@n23, Alpha=Get thigh_shape_bk_r() -> (thigh_shape_bk_r:double), Translation="5,-3,0", Rotation="0,0,0", Scale="1.5,1.5,1.5") {
    rotationMode=BMM_Additive
    rotationSpace=BCS_BoneSpace
    scaleMode=BMM_Additive
    scaleSpace=BCS_BoneSpace
    translationMode=BMM_Additive
    translationSpace=BCS_BoneSpace
    alphaBoolBlend {blendOption=Linear}
    boneToModify {boneName=thigh_volume_bk_r}
    scale=<1.5 1.5 1.5>
    translation=<5 -3>
   }
   n25 = "Transform (Modify) Bone - Bone: thigh_volume_lt_l"(ComponentPose=@n24, Alpha=Get thigh_shape_lt_l() -> (thigh_shape_lt_l:double), Translation="0,-5,0", Rotation="15,0,0", Scale="2,2,2") {
    rotationMode=BMM_Additive
    rotationSpace=BCS_BoneSpace
    scaleMode=BMM_Additive
    scaleSpace=BCS_BoneSpace
    translationMode=BMM_Additive
    translationSpace=BCS_BoneSpace
    alphaBoolBlend {blendOption=Linear}
    boneToModify {boneName=thigh_volume_lt_l}
    rotation {pitch=15}
    scale=<2 2 2>
    translation {y=-5}
   }
   n26 = "Transform (Modify) Bone - Bone: thigh_volume_rt_r"(ComponentPose=@n25, Alpha=Get thigh_shape_rt_r() -> (thigh_shape_rt_r:double), Translation="0,-5,0", Rotation="15,0,0", Scale="2,2,2") {
    rotationMode=BMM_Additive
    rotationSpace=BCS_BoneSpace
    scaleMode=BMM_Additive
    scaleSpace=BCS_BoneSpace
    translationMode=BMM_Additive
    translationSpace=BCS_BoneSpace
    alphaBoolBlend {blendOption=Linear}
    boneToModify {boneName=thigh_volume_rt_r}
    rotation {pitch=15}
    scale=<2 2 2>
    translation {y=-5}
   }
   n27 = "Transform (Modify) Bone - Bone: dyn_grenade_L01"(ComponentPose=@n26, Alpha=Get thigh_shape_lt_l(), Translation="1,-3,-2", Rotation="15,25,0", Scale="1,1,1") {
    rotationMode=BMM_Additive
    rotationSpace=BCS_BoneSpace
    scaleMode=BMM_Additive
    scaleSpace=BCS_BoneSpace
    translationMode=BMM_Additive
    translationSpace=BCS_BoneSpace
    alphaBoolBlend {blendOption=Linear}
    boneToModify {boneName=dyn_grenade_L01}
    rotation {pitch=15, yaw=25}
    translation=<1 -3 -2>
   }
   n28 = "Transform (Modify) Bone - Bone: dyn_grenade_R01"(ComponentPose=@n27, Alpha=Get thigh_shape_rt_r(), Translation="-2,3,2", Rotation="10,10,0", Scale="1,1,1") {
    rotationMode=BMM_Additive
    rotationSpace=BCS_BoneSpace
    scaleMode=BMM_Additive
    scaleSpace=BCS_BoneSpace
    translationMode=BMM_Additive
    translationSpace=BCS_BoneSpace
    alphaBoolBlend {blendOption=Linear}
    boneToModify {boneName=dyn_grenade_R01}
    rotation {pitch=10, yaw=10}
    translation=<-2 3 2>
   }
   n29 = "Transform (Modify) Bone - Bone: dyn_grenade_R01"(ComponentPose=@n28, Alpha=Get Shoulder_rot_f_r() -> (Shoulder_rot_f_r:double), Translation="0,0,0", Rotation="10,10,0", Scale="1,1,1") {
    rotationMode=BMM_Additive
    rotationSpace=BCS_BoneSpace
    scaleMode=BMM_Additive
    scaleSpace=BCS_BoneSpace
    translationMode=BMM_Additive
    translationSpace=BCS_BoneSpace
    alphaBoolBlend {blendOption=Linear}
    boneToModify {boneName=dyn_grenade_R01}
    rotation {pitch=10, yaw=10}
   }
   n30 = "Transform (Modify) Bone - Bone: dyn_grenade_L02"(ComponentPose=@n29, Alpha=Get thigh_shape_lt_l(), Translation="1,-1,0", Rotation="10,10,0", Scale="1,1,1") {
    rotationMode=BMM_Additive
    rotationSpace=BCS_BoneSpace
    scaleMode=BMM_Additive
    scaleSpace=BCS_BoneSpace
    translationMode=BMM_Additive
    translationSpace=BCS_BoneSpace
    alphaBoolBlend {blendOption=Linear}
    boneToModify {boneName=dyn_grenade_L02}
    rotation {pitch=10, yaw=10}
    translation=<1 -1>
   }
   n31 = "Transform (Modify) Bone - Bone: dyn_grenade_R02"(ComponentPose=@n30, Alpha=Get thigh_shape_rt_r(), Translation="-1.5,1,0", Rotation="10,10,0", Scale="1,1,1") {
    rotationMode=BMM_Additive
    rotationSpace=BCS_BoneSpace
    scaleMode=BMM_Additive
    scaleSpace=BCS_BoneSpace
    translationMode=BMM_Additive
    translationSpace=BCS_BoneSpace
    alphaBoolBlend {blendOption=Linear}
    boneToModify {boneName=dyn_grenade_R02}
    rotation {pitch=10, yaw=10}
    translation=<-1.5 1>
   }
   n32 = "Constraint - thigh_volume_fn_l (PP)"(ComponentPose=@n31, Alpha=Get thigh_shape_fn_l(), ConstraintWeights_0=0.5, ConstraintWeights_1=0.5) {
    alphaBoolBlend {blendOption=Linear}
    boneToModify {boneName=thigh_volume_fn_l}
    constraintSetup [{targetBone {boneName=thigh_volume_fn_l}}, {targetBone {boneName=thigh_volume_fn_r}}]
    constraintWeights [1,1]
   }
   n33 = "Constraint - thigh_volume_fn_r (PP)"(ComponentPose=@n32, Alpha=Get thigh_shape_fn_r(), ConstraintWeights_0=0.5, ConstraintWeights_1=0.5) {
    alphaBoolBlend {blendOption=Linear}
    boneToModify {boneName=thigh_volume_fn_r}
    constraintSetup [{targetBone {boneName=thigh_volume_fn_r}}, {targetBone {boneName=thigh_volume_fn_l}}]
    constraintWeights [1,1]
   }
   n34 = "Constraint - calf_l_front (RR)"!(ComponentPose=@n33, Alpha=1, ConstraintWeights_0=0.5, ConstraintWeights_1:float=0.5) -> (Pose) {
    alphaBoolBlend {blendOption=Linear}
    boneToModify {boneName=calf_l_front}
    constraintSetup [{transformType=Rotation, targetBone {boneName=thigh_l}}, {transformType=Rotation, targetBone {boneName=calf_l}}]
    constraintWeights [1,1]
   }
   n35 = "Constraint - calf_r_front (RR)"!(ComponentPose=@n34, Alpha=1, ConstraintWeights_0=0.5, ConstraintWeights_1:float=0.5) -> (Pose) {
    alphaBoolBlend {blendOption=Linear}
    boneToModify {boneName=calf_r_front}
    constraintSetup [{transformType=Rotation, targetBone {boneName=thigh_r}}, {transformType=Rotation, targetBone {boneName=calf_r}}]
    constraintWeights [1,1]
   }
   n36 = "Constraint - calf_l_back (RR)"!(ComponentPose=@n35, Alpha=1, ConstraintWeights_0=0.5, ConstraintWeights_1:float=0.5) -> (Pose) {
    alphaBoolBlend {blendOption=Linear}
    boneToModify {boneName=calf_l_back}
    constraintSetup [{transformType=Rotation, targetBone {boneName=thigh_l}}, {transformType=Rotation, targetBone {boneName=calf_l}}]
    constraintWeights [1,1]
   }
   n37 = "Constraint - calf_r_back (RR)"!(ComponentPose=@n36, Alpha=1, ConstraintWeights_0=0.5, ConstraintWeights_1:float=0.5) -> (Pose) {
    alphaBoolBlend {blendOption=Linear}
    boneToModify {boneName=calf_r_back}
    constraintSetup [{transformType=Rotation, targetBone {boneName=thigh_r}}, {transformType=Rotation, targetBone {boneName=calf_r}}]
    constraintWeights [1,1]
   }
   n38 = Get calf_l_front() -> (calf_l_front:double)
   n39 = "Transform (Modify) Bone - Bone: calf_l_front"(ComponentPose=@n37, Alpha=@n38, Translation="10,-2,0", Rotation="0,0,0", Scale="1,1,1") {
    rotationMode=BMM_Additive
    rotationSpace=BCS_BoneSpace
    scaleMode=BMM_Additive
    scaleSpace=BCS_BoneSpace
    translationMode=BMM_Additive
    translationSpace=BCS_BoneSpace
    alphaBoolBlend {blendOption=Linear}
    boneToModify {boneName=calf_l_front}
    translation=<10 -2>
   }
   n40 = Get calf_r_front() -> (calf_r_front:double)
   n41 = "Transform (Modify) Bone - Bone: calf_r_front"(ComponentPose=@n39, Alpha=@n40, Translation="-10,2,0", Rotation="0,0,0", Scale="1,1,1") {
    rotationMode=BMM_Additive
    rotationSpace=BCS_BoneSpace
    scaleMode=BMM_Additive
    scaleSpace=BCS_BoneSpace
    translationMode=BMM_Additive
    translationSpace=BCS_BoneSpace
    alphaBoolBlend {blendOption=Linear}
    boneToModify {boneName=calf_r_front}
    translation=<-10 2>
   }
   n42 = "Transform (Modify) Bone - Bone: calf_l_back"(ComponentPose=@n41, Alpha=@n38, Translation="0,-12,0", Rotation="0,0,0", Scale="1,1,1") {
    rotationMode=BMM_Additive
    rotationSpace=BCS_BoneSpace
    scaleMode=BMM_Additive
    scaleSpace=BCS_BoneSpace
    translationMode=BMM_Additive
    translationSpace=BCS_BoneSpace
    alphaBoolBlend {blendOption=Linear}
    boneToModify {boneName=calf_l_back}
    translation {y=-12}
   }
   n43 = "Transform (Modify) Bone - Bone: calf_r_back"(ComponentPose=@n42, Alpha=@n40, Translation="0,12,0", Rotation="0,0,0", Scale="1,1,1") {
    rotationMode=BMM_Additive
    rotationSpace=BCS_BoneSpace
    scaleMode=BMM_Additive
    scaleSpace=BCS_BoneSpace
    translationMode=BMM_Additive
    translationSpace=BCS_BoneSpace
    alphaBoolBlend {blendOption=Linear}
    boneToModify {boneName=calf_r_back}
    translation {y=12}
   }
   n44 = "Transform (Modify) Bone - Bone: calf_l_back01"(ComponentPose=@n43, Alpha=@n38, Translation="15,0,0", Rotation="0,0,0", Scale="1,1,1") {
    rotationMode=BMM_Additive
    rotationSpace=BCS_BoneSpace
    scaleMode=BMM_Additive
    scaleSpace=BCS_BoneSpace
    translationMode=BMM_Additive
    translationSpace=BCS_BoneSpace
    alphaBoolBlend {blendOption=Linear}
    boneToModify {boneName=calf_l_back01}
    translation {x=15}
   }
   n45 = "Transform (Modify) Bone - Bone: calf_r_back01"(ComponentPose=@n44, Alpha=@n40, Translation="-15,0,0", Rotation="0,0,0", Scale="1,1,1") {
    rotationMode=BMM_Additive
    rotationSpace=BCS_BoneSpace
    scaleMode=BMM_Additive
    scaleSpace=BCS_BoneSpace
    translationMode=BMM_Additive
    translationSpace=BCS_BoneSpace
    alphaBoolBlend {blendOption=Linear}
    boneToModify {boneName=calf_r_back01}
    translation {x=-15}
   }
   Output Pose(Result=Component To Local(ComponentPose=@n45)) {layerGroup=None}
   n46 = Get isFrontEnd() -> (isFrontEnd:double)
   n47 = "Transform (Modify) Bone - Bone: dyn_shoulderPlate_r"(
    ComponentPose=Local To Component(LocalPose=Copy Pose From Mesh() {bUseAttachedParent=true})
    Alpha=@n46
    Translation="0,0,0"
    Rotation="0,0,0"
    Scale="0.9,0.9,0.9"
   ) {
    scaleMode=BMM_Additive
    scaleSpace=BCS_BoneSpace
    alphaBoolBlend {blendOption=Linear}
    boneToModify {boneName=dyn_shoulderPlate_r}
    scale=<0.9 0.9 0.9>
   }
   n48 = "Transform (Modify) Bone - Bone: dyn_shoulderPlate_l"(ComponentPose=@n47, Alpha=1, Translation="0,0,0", Rotation="0,0,0", Scale="0.9,0.9,0.9") {
    scaleMode=BMM_Additive
    scaleSpace=BCS_BoneSpace
    alphaBoolBlend {blendOption=Linear}
    boneToModify {boneName=dyn_shoulderPlate_l}
    scale=<0.9 0.9 0.9>
   }
   n49 = "Transform (Modify) Bone - Bone: dyn_shoulderPlate_r"(ComponentPose=@n48, Alpha=@n46, Translation="0,0,0", Rotation="0,0,0", Scale="1.1,1.1,1.1") {
    scaleMode=BMM_Additive
    scaleSpace=BCS_BoneSpace
    translationMode=BMM_Additive
    alphaBoolBlend {blendOption=Linear}
    boneToModify {boneName=dyn_shoulderPlate_r}
    scale=<1.1 1.1 1.1>
   }
   n50 = "Transform (Modify) Bone - Bone: dyn_shoulderPlate_l"(ComponentPose=@n49, Alpha=1, Translation="0,1,0", Rotation="0,0,0", Scale="1.1,1.1,1.1") {
    scaleMode=BMM_Additive
    scaleSpace=BCS_BoneSpace
    translationMode=BMM_Additive
    alphaBoolBlend {blendOption=Linear}
    boneToModify {boneName=dyn_shoulderPlate_l}
    scale=<1.1 1.1 1.1>
    translation {y=1}
   }
   n51 = "Constraint - attach_cape (PP)"!(ComponentPose=@n50, Alpha=1, ConstraintWeights_0=0.8, ConstraintWeights_1:float=0.2) -> (Pose) {
    alphaBoolBlend {blendOption=Linear}
    boneToModify {boneName=attach_cape}
    constraintSetup [{targetBone {boneName=spine_05}}, {targetBone {boneName=spine_04}}]
    constraintWeights [1,1]
   }
   n52 = "Constraint - attach_backpack (PP)"!(ComponentPose=@n51, Alpha=1, ConstraintWeights_0=0.7, ConstraintWeights_1:float=0.3) -> (Pose) {
    alphaBoolBlend {blendOption=Linear}
    boneToModify {boneName=attach_backpack}
    constraintSetup [{targetBone {boneName=spine_05}}, {targetBone {boneName=spine_04}}]
    constraintWeights [1,1]
   }
   n53 = "Constraint - dyn_arm_l (PP)"!(ComponentPose=@n52, Alpha=1, ConstraintWeights_0=0.5, ConstraintWeights_1:float=0.5) -> (Pose) {
    alphaBoolBlend {blendOption=Linear}
    boneToModify {boneName=dyn_arm_l}
    constraintSetup [{targetBone {boneName=clavicle_l}}, {targetBone {boneName=spine_05}}]
    constraintWeights [1,1]
   }
   n54 = "Constraint - dyn_arm_r (PP)"!(ComponentPose=@n53, Alpha=1, ConstraintWeights_0=0.5, ConstraintWeights_1:float=0.5) -> (Pose) {
    alphaBoolBlend {blendOption=Linear}
    boneToModify {boneName=dyn_arm_r}
    constraintSetup [{targetBone {boneName=clavicle_r}}, {targetBone {boneName=spine_05}}]
    constraintWeights [1,1]
   }
   n55 = "Constraint - dyn_arm_aim_l (RR)"!(ComponentPose=@n54, Alpha=1, ConstraintWeights_0=0.7, ConstraintWeights_1:float=0.3) -> (Pose) {
    alphaBoolBlend {blendOption=Linear}
    boneToModify {boneName=dyn_arm_aim_l}
    constraintSetup [{transformType=Rotation, targetBone {boneName=upperarm_l}}, {transformType=Rotation, targetBone {boneName=dyn_arm_l}}]
    constraintWeights [1,1]
   }
   n56 = "Constraint - dyn_arm_aim_r (RR)"!(ComponentPose=@n55, Alpha=1, ConstraintWeights_0=0.7, ConstraintWeights_1:float=0.3) -> (Pose) {
    alphaBoolBlend {blendOption=Linear}
    boneToModify {boneName=dyn_arm_aim_r}
    constraintSetup [{transformType=Rotation, targetBone {boneName=upperarm_r}}, {transformType=Rotation, targetBone {boneName=dyn_arm_r}}]
    constraintWeights [1,1]
   }
   n57 = "Constraint - dyn_lowerarm_base_l (PP)"!(ComponentPose=@n56, Alpha=1, ConstraintWeights_0=1, ConstraintWeights_1:float=0) -> (Pose) {
    alphaBoolBlend {blendOption=Linear}
    boneToModify {boneName=dyn_lowerarm_base_l}
    constraintSetup [{targetBone {boneName=lowerarm_l}}, {targetBone {boneName=dyn_arm_aim_l}}]
    constraintWeights [1,1]
   }
   n58 = "Constraint - dyn_lowerarm_base_r (PP)"!(ComponentPose=@n57, Alpha=1, ConstraintWeights_0=1, ConstraintWeights_1:float=0) -> (Pose) {
    alphaBoolBlend {blendOption=Linear}
    boneToModify {boneName=dyn_lowerarm_base_r}
    constraintSetup [{targetBone {boneName=lowerarm_r}}, {targetBone {boneName=dyn_arm_aim_r}}]
    constraintWeights [1,1]
   }
   n59 = "Constraint - dyn_lowerarm_base_l (RR)"!(ComponentPose=@n58, Alpha=1, ConstraintWeights_0=0.7, ConstraintWeights_1:float=0.3) -> (Pose) {
    alphaBoolBlend {blendOption=Linear}
    boneToModify {boneName=dyn_lowerarm_base_l}
    constraintSetup [{transformType=Rotation, targetBone {boneName=upperarm_l}}, {transformType=Rotation, targetBone {boneName=upperarm_twist_02_l}}]
    constraintWeights [1,1]
   }
   n60 = "Constraint - dyn_lowerarm_base_r (RR)"!(ComponentPose=@n59, Alpha=1, ConstraintWeights_0=0.7, ConstraintWeights_1:float=0.3) -> (Pose) {
    alphaBoolBlend {blendOption=Linear}
    boneToModify {boneName=dyn_lowerarm_base_r}
    constraintSetup [{transformType=Rotation, targetBone {boneName=upperarm_r}}, {transformType=Rotation, targetBone {boneName=dyn_arm_aim_r}}]
    constraintWeights [1,1]
   }
   n61 = "Constraint - dyn_arm_aim_twist_01_l (PP)"!(ComponentPose=@n60, Alpha=1, ConstraintWeights_0=0, ConstraintWeights_1:float=1) -> (Pose) {
    alphaBoolBlend {blendOption=Linear}
    boneToModify {boneName=dyn_arm_aim_twist_01_l}
    constraintSetup [{targetBone {boneName=dyn_arm_aim_twist_01_l}}, {targetBone {boneName=dyn_lowerarm_base_l}}]
    constraintWeights [1,1]
   }
   n62 = "Constraint - dyn_arm_aim_twist_01_r (PP)"!(ComponentPose=@n61, Alpha=1, ConstraintWeights_0=0, ConstraintWeights_1:float=1) -> (Pose) {
    alphaBoolBlend {blendOption=Linear}
    boneToModify {boneName=dyn_arm_aim_twist_01_r}
    constraintSetup [{targetBone {boneName=dyn_arm_aim_twist_01_r}}, {targetBone {boneName=dyn_lowerarm_base_r}}]
    constraintWeights [1,1]
   }
   n63 = "Constraint - dyn_arm_aim_twist_02_l (PP)"!(ComponentPose=@n62, Alpha=1, ConstraintWeights_0=0, ConstraintWeights_1:float=1) -> (Pose) {
    alphaBoolBlend {blendOption=Linear}
    boneToModify {boneName=dyn_arm_aim_twist_02_l}
    constraintSetup [{targetBone {boneName=dyn_arm_aim_twist_02_l}}, {targetBone {boneName=dyn_lowerarm_base_l}}]
    constraintWeights [1,1]
   }
   n64 = "Constraint - dyn_arm_aim_twist_02_r (PP)"!(ComponentPose=@n63, Alpha=1, ConstraintWeights_0=0, ConstraintWeights_1:float=1) -> (Pose) {
    alphaBoolBlend {blendOption=Linear}
    boneToModify {boneName=dyn_arm_aim_twist_02_r}
    constraintSetup [{targetBone {boneName=dyn_arm_aim_twist_02_r}}, {targetBone {boneName=dyn_lowerarm_base_r}}]
    constraintWeights [1,1]
   }
   n65 = "Constraint - dyn_arm_aim_twist_01_l (RR)"!(ComponentPose=@n64, Alpha=1, ConstraintWeights_0=0.9, ConstraintWeights_1:float=0.1) -> (Pose) {
    alphaBoolBlend {blendOption=Linear}
    boneToModify {boneName=dyn_arm_aim_twist_01_l}
    constraintSetup [{transformType=Rotation, targetBone {boneName=upperarm_twist_01_l}}, {transformType=Rotation, targetBone {boneName=upperarm_twist_02_l}}]
    constraintWeights [1,1]
   }
   n66 = "Constraint - dyn_arm_aim_twist_01_r (RR)"!(ComponentPose=@n65, Alpha=1, ConstraintWeights_0=0.9, ConstraintWeights_1:float=0.1) -> (Pose) {
    alphaBoolBlend {blendOption=Linear}
    boneToModify {boneName=dyn_arm_aim_twist_01_r}
    constraintSetup [{transformType=Rotation, targetBone {boneName=upperarm_twist_01_r}}, {transformType=Rotation, targetBone {boneName=dyn_lowerarm_base_r}}]
    constraintWeights [1,1]
   }
   n67 = "Constraint - dyn_arm_aim_twist_02_l (RR)"!(ComponentPose=@n66, Alpha=1, ConstraintWeights_0=0.9, ConstraintWeights_1:float=0.1) -> (Pose) {
    alphaBoolBlend {blendOption=Linear}
    boneToModify {boneName=dyn_arm_aim_twist_02_l}
    constraintSetup [{transformType=Rotation, targetBone {boneName=upperarm_twist_02_l}}, {transformType=Rotation, targetBone {boneName=upperarm_l}}]
    constraintWeights [1,1]
   }
   n68 = "Constraint - dyn_arm_aim_twist_02_r (RR)"!(ComponentPose=@n67, Alpha=1, ConstraintWeights_0=0.9, ConstraintWeights_1:float=0.1) -> (Pose) {
    alphaBoolBlend {blendOption=Linear}
    boneToModify {boneName=dyn_arm_aim_twist_02_r}
    constraintSetup [{transformType=Rotation, targetBone {boneName=upperarm_twist_02_r}}, {transformType=Rotation, targetBone {boneName=dyn_lowerarm_base_r}}]
    constraintWeights [1,1]
   }
   n69 = Get Shoulder_rot_f_l() -> (Shoulder_rot_f_l:double)
   n70 = "Transform (Modify) Bone - Bone: dyn_arm_aim_twist_01_l"(ComponentPose=@n68, Alpha=@n69, Translation="5,0,0", Rotation="0,0,-15", Scale="1,1,1") {
    rotationMode=BMM_Additive
    rotationSpace=BCS_BoneSpace
    scaleMode=BMM_Additive
    scaleSpace=BCS_BoneSpace
    translationMode=BMM_Additive
    translationSpace=BCS_BoneSpace
    alphaBoolBlend {blendOption=Linear}
    boneToModify {boneName=dyn_arm_aim_twist_01_l}
    rotation {roll=-15}
    translation {x=5}
   }
   n71 = "Transform (Modify) Bone - Bone: dyn_arm_aim_twist_02_l"(ComponentPose=@n70, Alpha=@n69, Translation="5,0,0", Rotation="0,0,-15", Scale="1,1,1") {
    rotationMode=BMM_Additive
    rotationSpace=BCS_BoneSpace
    scaleMode=BMM_Additive
    scaleSpace=BCS_BoneSpace
    translationMode=BMM_Additive
    translationSpace=BCS_BoneSpace
    alphaBoolBlend {blendOption=Linear}
    boneToModify {boneName=dyn_arm_aim_twist_02_l}
    rotation {roll=-15}
    translation {x=5}
   }
   n72 = Get Shoulder_rot_f_r()
   n73 = "Transform (Modify) Bone - Bone: dyn_arm_aim_twist_01_r"(ComponentPose=@n71, Alpha=@n72, Translation="-5,0,0", Rotation="0,0,-15", Scale="1.2,1.2,1.2") {
    rotationMode=BMM_Additive
    rotationSpace=BCS_BoneSpace
    scaleMode=BMM_Additive
    scaleSpace=BCS_BoneSpace
    translationMode=BMM_Additive
    translationSpace=BCS_BoneSpace
    alphaBoolBlend {blendOption=Linear}
    boneToModify {boneName=dyn_arm_aim_twist_01_r}
    rotation {roll=-15}
    scale=<1.2 1.2 1.2>
    translation {x=-5}
   }
   n74 = "Transform (Modify) Bone - Bone: dyn_arm_aim_twist_02_r"(ComponentPose=@n73, Alpha=@n72, Translation="-5,0,0", Rotation="0,0,-15", Scale="1,1,1") {
    rotationMode=BMM_Additive
    rotationSpace=BCS_BoneSpace
    scaleMode=BMM_Additive
    scaleSpace=BCS_BoneSpace
    translationMode=BMM_Additive
    translationSpace=BCS_BoneSpace
    alphaBoolBlend {blendOption=Linear}
    boneToModify {boneName=dyn_arm_aim_twist_02_r}
    rotation {roll=-15}
    translation {x=-5}
   }
   n75 = "Constraint - back_uppelalm_l (PP)"!(ComponentPose=@n74, Alpha=1, ConstraintWeights_0=0, ConstraintWeights_1:float=1) -> (Pose) {
    alphaBoolBlend {blendOption=Linear}
    boneToModify {boneName=back_uppelalm_l}
    constraintSetup [{targetBone {boneName=upperarm_twist_01_l}}, {targetBone {boneName=dyn_arm_aim_twist_01_l}}]
    constraintWeights [1,1]
   }
   n76 = "Transform (Modify) Bone - Bone: back_uppelalm_l"(ComponentPose=@n75, Alpha=@n69, Translation="0,0,0", Rotation="0,0,-7", Scale="1.15,1.7,1.15") {
    rotationMode=BMM_Additive
    rotationSpace=BCS_BoneSpace
    scaleMode=BMM_Additive
    scaleSpace=BCS_BoneSpace
    translationMode=BMM_Additive
    translationSpace=BCS_BoneSpace
    alphaBoolBlend {blendOption=Linear}
    boneToModify {boneName=back_uppelalm_l}
    rotation {roll=-7}
    scale=<1.15 1.7 1.15>
   }
   n77 = "Constraint - back_upperarm_r (PP)"!(ComponentPose=@n76, Alpha=1, ConstraintWeights_0=0, ConstraintWeights_1:float=1) -> (Pose) {
    alphaBoolBlend {blendOption=Linear}
    boneToModify {boneName=back_upperarm_r}
    constraintSetup [{targetBone {boneName=upperarm_twist_01_r}}, {targetBone {boneName=dyn_arm_aim_twist_01_r}}]
    constraintWeights [1,1]
   }
   n78 = "Transform (Modify) Bone - Bone: back_upperarm_r"(ComponentPose=@n77, Alpha=@n72, Translation="0,0,0", Rotation="0,0,0", Scale="1.15,1.5,1.15") {
    rotationMode=BMM_Additive
    rotationSpace=BCS_BoneSpace
    scaleMode=BMM_Additive
    scaleSpace=BCS_BoneSpace
    translationMode=BMM_Additive
    translationSpace=BCS_BoneSpace
    alphaBoolBlend {blendOption=Linear}
    boneToModify {boneName=back_upperarm_r}
    scale=<1.15 1.5 1.15>
   }
   n79 = "Constraint - dyn_shoulderPlate_base_l (PP)"!(ComponentPose=@n78, Alpha=1, ConstraintWeights_0=0.9, ConstraintWeights_1:float=0.1) -> (Pose) {
    alphaBoolBlend {blendOption=Linear}
    boneToModify {boneName=dyn_shoulderPlate_base_l}
    constraintSetup [{targetBone {boneName=clavicle_l}}, {targetBone {boneName=dyn_arm_aim_l}}]
    constraintWeights [1,1]
   }
   n80 = "Constraint - dyn_shoulderPlate_base_l (PP)"(ComponentPose=@n79, Alpha=Get IsSkyDiving_L() -> (IsSkyDiving_L:double), ConstraintWeights_0=0.5, ConstraintWeights_1=0.5) {
    alphaBoolBlend {blendOption=Linear}
    boneToModify {boneName=dyn_shoulderPlate_base_l}
    constraintSetup [{targetBone {boneName=clavicle_l}}, {targetBone {boneName=dyn_arm_aim_l}}]
    constraintWeights [1,1]
   }
   n81 = Get Shoulder_rot_up_l() -> (Shoulder_rot_up_l:double)
   n82 = "Transform (Modify) Bone - Bone: dyn_shoulderPlate_base_l"(ComponentPose=@n80, Alpha=@n81, Translation="1,-2,1", Rotation="60,-60,-60", Scale="1,1,1") {
    rotationMode=BMM_Additive
    rotationSpace=BCS_BoneSpace
    scaleMode=BMM_Additive
    scaleSpace=BCS_BoneSpace
    translationMode=BMM_Additive
    translationSpace=BCS_BoneSpace
    alphaBoolBlend {blendOption=Linear}
    boneToModify {boneName=dyn_shoulderPlate_base_l}
    rotation=<rot 60 -60 -60>
    translation=<1 -2 1>
   }
   n83 = "Transform (Modify) Bone - Bone: dyn_shoulderPlate_l"(ComponentPose=@n82, Alpha=@n81, Translation="0,0,0", Rotation="0,15,10", Scale="1,1,1") {
    rotationMode=BMM_Additive
    rotationSpace=BCS_BoneSpace
    scaleMode=BMM_Additive
    scaleSpace=BCS_BoneSpace
    translationMode=BMM_Additive
    translationSpace=BCS_BoneSpace
    alphaBoolBlend {blendOption=Linear}
    boneToModify {boneName=dyn_shoulderPlate_l}
    rotation {roll=10, yaw=15}
   }
   n84 = "Transform (Modify) Bone - Bone: dyn_shoulderPlate_base_l"(
    ComponentPose=@n83
    Alpha=Get Shoulder_rot_back_l() -> (Shoulder_rot_back_l:double)
    Translation="-3,-6.5,-5"
    Rotation="-10,-15,-40"
    Scale="1,1.3,1"
   ) {
    rotationMode=BMM_Additive
    rotationSpace=BCS_BoneSpace
    scaleMode=BMM_Additive
    scaleSpace=BCS_BoneSpace
    translationMode=BMM_Additive
    translationSpace=BCS_BoneSpace
    alphaBoolBlend {blendOption=Linear}
    boneToModify {boneName=dyn_shoulderPlate_base_l}
    rotation=<rot -10 -15 -40>
    scale {y=1.3}
    translation=<-3 -6.5 -5>
   }
   n85 = "Constraint - dyn_shoulderPlate_base_r (PP)"!(ComponentPose=@n84, Alpha=1, ConstraintWeights_0=0.8, ConstraintWeights_1:float=0.2) -> (Pose) {
    alphaBoolBlend {blendOption=Linear}
    boneToModify {boneName=dyn_shoulderPlate_base_r}
    constraintSetup [{targetBone {boneName=clavicle_r}}, {targetBone {boneName=dyn_arm_aim_r}}]
    constraintWeights [1,1]
   }
   n86 = "Constraint - dyn_shoulderPlate_base_r (PP)"(ComponentPose=@n85, Alpha=Get IsSkyDiving_R() -> (IsSkyDiving_R:double), ConstraintWeights_0=0.5, ConstraintWeights_1=0.5) {
    alphaBoolBlend {blendOption=Linear}
    boneToModify {boneName=dyn_shoulderPlate_base_r}
    constraintSetup [{targetBone {boneName=clavicle_r}}, {targetBone {boneName=dyn_arm_aim_r}}]
    constraintWeights [1,1]
   }
   n87 = Get Shoulder_rot_up_r() -> (Shoulder_rot_up_r:double)
   n88 = "Transform (Modify) Bone - Bone: dyn_shoulderPlate_base_r"(ComponentPose=@n86, Alpha=@n87, Translation="1,1,0.5", Rotation="60,25,30", Scale="1,1,1") {
    rotationMode=BMM_Additive
    rotationSpace=BCS_BoneSpace
    scaleMode=BMM_Additive
    scaleSpace=BCS_BoneSpace
    translationMode=BMM_Additive
    translationSpace=BCS_BoneSpace
    alphaBoolBlend {blendOption=Linear}
    boneToModify {boneName=dyn_shoulderPlate_base_r}
    rotation=<rot 60 25 30>
    translation=<1 1 0.5>
   }
   n89 = "Transform (Modify) Bone - Bone: dyn_shoulderPlate_r"(ComponentPose=@n88, Alpha=@n87, Translation="0,0,0", Rotation="0,-7.5,-5", Scale="1,1,1") {
    rotationMode=BMM_Additive
    rotationSpace=BCS_BoneSpace
    scaleMode=BMM_Additive
    scaleSpace=BCS_BoneSpace
    translationMode=BMM_Additive
    translationSpace=BCS_BoneSpace
    alphaBoolBlend {blendOption=Linear}
    boneToModify {boneName=dyn_shoulderPlate_r}
    rotation {roll=-5, yaw=-7.5}
   }
   n90 = "Transform (Modify) Bone - Bone: dyn_shoulderPlate_base_r"(ComponentPose=@n89, Alpha=Get Shoulder_rot_back_r() -> (Shoulder_rot_back_r:double), Translation="0,0,0", Rotation="0,0,30", Scale="1,1.2,1") {
    rotationMode=BMM_Additive
    rotationSpace=BCS_BoneSpace
    scaleMode=BMM_Additive
    scaleSpace=BCS_BoneSpace
    translationMode=BMM_Additive
    translationSpace=BCS_BoneSpace
    alphaBoolBlend {blendOption=Linear}
    boneToModify {boneName=dyn_shoulderPlate_base_r}
    rotation {roll=30}
    scale {y=1.2}
   }
   n91 = "Transform (Modify) Bone - Bone: dyn_shoulderPlate_base_r"(ComponentPose=@n90, Alpha=Get Shoulder_rot_dn_r() -> (Shoulder_rot_dn_r:double), Translation="-1.5,1.5,0", Rotation="0,0,0", Scale="1,1,1") {
    rotationMode=BMM_Additive
    rotationSpace=BCS_BoneSpace
    scaleMode=BMM_Additive
    scaleSpace=BCS_BoneSpace
    translationMode=BMM_Additive
    translationSpace=BCS_BoneSpace
    alphaBoolBlend {blendOption=Linear}
    boneToModify {boneName=dyn_shoulderPlate_base_r}
    translation=<-1.5 1.5>
   }
   n92 = Get offset_weapon() -> (offset_weapon:double)
   n93 = "Transform (Modify) Bone - Bone: dyn_shoulderPlate_base_r"(ComponentPose=@n91, Alpha=@n92, Translation="0,6.5,2", Rotation="-10,20,0", Scale="0.8,1,0.85") {
    rotationMode=BMM_Additive
    rotationSpace=BCS_BoneSpace
    scaleMode=BMM_Additive
    scaleSpace=BCS_BoneSpace
    translationMode=BMM_Additive
    translationSpace=BCS_BoneSpace
    alphaBoolBlend {blendOption=Linear}
    boneToModify {boneName=dyn_shoulderPlate_base_r}
    rotation {pitch=-10, yaw=20}
    scale {x=0.8, z=0.85}
    translation {y=6.5, z=2}
   }
   n94 = "Transform (Modify) Bone - Bone: dyn_shoulderPlate_base_r"(ComponentPose=@n93, Alpha=Get offset_weapon_1() -> (offset_weapon_1:double), Translation="-6,-5,-2.5", Rotation="-30,25,35", Scale="1,1,1") {
    rotationMode=BMM_Additive
    rotationSpace=BCS_BoneSpace
    scaleMode=BMM_Additive
    scaleSpace=BCS_BoneSpace
    translationMode=BMM_Additive
    translationSpace=BCS_BoneSpace
    alphaBoolBlend {blendOption=Linear}
    boneToModify {boneName=dyn_shoulderPlate_base_r}
    rotation=<rot -30 25 35>
    translation=<-6 -5 -2.5>
   }
   n95 = "Transform (Modify) Bone - Bone: dyn_shoulderPlate_base_r"(
    ComponentPose=@n94
    Alpha=Get Shoulder_rot_up_r_1() -> (Shoulder_rot_up_r_1:double)
    Translation="-1,5,-5"
    Rotation="-10,18,0"
    Scale="1,1.2,1"
   ) {
    rotationMode=BMM_Additive
    rotationSpace=BCS_BoneSpace
    scaleMode=BMM_Additive
    scaleSpace=BCS_BoneSpace
    translationMode=BMM_Additive
    translationSpace=BCS_BoneSpace
    alphaBoolBlend {blendOption=Linear}
    boneToModify {boneName=dyn_shoulderPlate_base_r}
    rotation {pitch=-10, yaw=18}
    scale {y=1.2}
    translation=<-1 5 -5>
   }
   n96 = "Transform (Modify) Bone - Bone: dyn_arm_aim_twist_01_r"(ComponentPose=@n95, Alpha=@n92, Translation="0,0,0", Rotation="0,0,-15", Scale="1.15,1.15,1.15") {
    rotationMode=BMM_Additive
    rotationSpace=BCS_BoneSpace
    scaleMode=BMM_Additive
    scaleSpace=BCS_BoneSpace
    translationMode=BMM_Additive
    translationSpace=BCS_BoneSpace
    alphaBoolBlend {blendOption=Linear}
    boneToModify {boneName=dyn_arm_aim_twist_01_r}
    rotation {roll=-15}
    scale=<1.15 1.15 1.15>
   }
   n97 = "Transform (Modify) Bone - Bone: dyn_shoulderPlate_base_l"(
    ComponentPose=@n96
    Alpha=Get offset_weapon_l_Shoulder() -> (offset_weapon_l_Shoulder:double)
    Translation="3,-15,-10"
    Rotation="-150,-110,0"
    Scale="1,1,1"
   ) {
    rotationMode=BMM_Additive
    rotationSpace=BCS_BoneSpace
    scaleMode=BMM_Additive
    scaleSpace=BCS_BoneSpace
    translationMode=BMM_Additive
    translationSpace=BCS_BoneSpace
    alphaBoolBlend {blendOption=Linear}
    boneToModify {boneName=dyn_shoulderPlate_base_l}
    rotation {pitch=-150, yaw=-110}
    translation=<3 -15 -10>
   }
   n98 = "Transform (Modify) Bone - Bone: dyn_shoulderPlate_base_l"(
    ComponentPose=@n97
    Alpha=Get offset_weapon_l_Shoulder1() -> (offset_weapon_l_Shoulder1:double)
    Translation="-3,-15,0"
    Rotation="-100,-130,60"
    Scale="1,1,1"
   ) {
    rotationMode=BMM_Additive
    rotationSpace=BCS_BoneSpace
    scaleMode=BMM_Additive
    scaleSpace=BCS_BoneSpace
    translationMode=BMM_Additive
    translationSpace=BCS_BoneSpace
    alphaBoolBlend {blendOption=Linear}
    boneToModify {boneName=dyn_shoulderPlate_base_l}
    rotation=<rot -100 -130 60>
    translation=<-3 -15>
   }
   n99 = "Transform (Modify) Bone - Bone: upperarmDn_r"(ComponentPose=@n98, Alpha=Get Shoulder_rot_dn_r(), Translation="5,0,0", Rotation="-10,0,0", Scale="1,1,1") {
    rotationMode=BMM_Additive
    rotationSpace=BCS_BoneSpace
    scaleMode=BMM_Additive
    scaleSpace=BCS_BoneSpace
    translationMode=BMM_Additive
    translationSpace=BCS_BoneSpace
    alphaBoolBlend {blendOption=Linear}
    boneToModify {boneName=upperarmDn_r}
    rotation {pitch=-10}
    translation {x=5}
   }
   n100 = "Transform (Modify) Bone - Bone: upperarmDn_l"(ComponentPose=@n99, Alpha=Get Shoulder_rot_dn_l() -> (Shoulder_rot_dn_l:double), Translation="-5,0,0", Rotation="-10,0,0", Scale="1,1,1") {
    rotationMode=BMM_Additive
    rotationSpace=BCS_BoneSpace
    scaleMode=BMM_Additive
    scaleSpace=BCS_BoneSpace
    translationMode=BMM_Additive
    translationSpace=BCS_BoneSpace
    alphaBoolBlend {blendOption=Linear}
    boneToModify {boneName=upperarmDn_l}
    rotation {pitch=-10}
    translation {x=-5}
   }
   n101 = Get Shoulder_rot_f_r()
   n102 = "Transform (Modify) Bone - Bone: upperarmFn_r"(ComponentPose=@n100, Alpha=@n101, Translation="10,0,0", Rotation="0,0,0", Scale="1,1,1") {
    rotationMode=BMM_Additive
    rotationSpace=BCS_BoneSpace
    scaleMode=BMM_Additive
    scaleSpace=BCS_BoneSpace
    translationMode=BMM_Additive
    translationSpace=BCS_BoneSpace
    alphaBoolBlend {blendOption=Linear}
    boneToModify {boneName=upperarmFn_r}
    translation {x=10}
   }
   n103 = Get Shoulder_rot_f_l()
   n104 = "Transform (Modify) Bone - Bone: upperarmFn_l"(ComponentPose=@n102, Alpha=@n103, Translation="-10,0,0", Rotation="0,0,0", Scale="1,1,1") {
    rotationMode=BMM_Additive
    rotationSpace=BCS_BoneSpace
    scaleMode=BMM_Additive
    scaleSpace=BCS_BoneSpace
    translationMode=BMM_Additive
    translationSpace=BCS_BoneSpace
    alphaBoolBlend {blendOption=Linear}
    boneToModify {boneName=upperarmFn_l}
    translation {x=-10}
   }
   n105 = "Transform (Modify) Bone - Bone: dyn_arm_r"(ComponentPose=@n104, Alpha=@n101, Translation="-3,0,0", Rotation="0,0,0", Scale="1,1,1") {
    rotationMode=BMM_Additive
    rotationSpace=BCS_BoneSpace
    scaleMode=BMM_Additive
    scaleSpace=BCS_BoneSpace
    translationMode=BMM_Additive
    translationSpace=BCS_BoneSpace
    alphaBoolBlend {blendOption=Linear}
    boneToModify {boneName=dyn_arm_r}
    translation {x=-3}
   }
   n106 = "Transform (Modify) Bone - Bone: dyn_shoulderPlate_base_l"(
    ComponentPose=@n105
    Alpha=Get Shoulder_rot_back_l_0() -> (Shoulder_rot_back_l_0:double)
    Translation="0,0,0"
    Rotation="0,0,20"
    Scale="1,1,1"
   ) {
    rotationMode=BMM_Additive
    rotationSpace=BCS_BoneSpace
    scaleMode=BMM_Additive
    scaleSpace=BCS_BoneSpace
    translationMode=BMM_Additive
    translationSpace=BCS_BoneSpace
    alphaBoolBlend {blendOption=Linear}
    boneToModify {boneName=dyn_shoulderPlate_base_l}
    rotation {roll=20}
   }
   n107 = "Transform (Modify) Bone - Bone: dyn_shoulderPlate_base_l"(ComponentPose=@n106, Alpha=Get IsSkyDiving_L(), Translation="-5,7,-5", Rotation="35,10,5", Scale="0.8,1,0.9") {
    rotationMode=BMM_Additive
    rotationSpace=BCS_BoneSpace
    scaleMode=BMM_Additive
    scaleSpace=BCS_BoneSpace
    translationMode=BMM_Additive
    translationSpace=BCS_BoneSpace
    alphaBoolBlend {blendOption=Linear}
    boneToModify {boneName=dyn_shoulderPlate_base_l}
    rotation=<rot 35 10 5>
    scale {x=0.8, z=0.9}
    translation=<-5 7 -5>
   }
   n108 = "Transform (Modify) Bone - Bone: dyn_shoulderPlate_r"(ComponentPose=@n107, Alpha=Get IsSkyDiving_R(), Translation="-1,0,0", Rotation="15,-15,-5", Scale="0.8,1,0.9") {
    rotationMode=BMM_Additive
    rotationSpace=BCS_BoneSpace
    scaleMode=BMM_Additive
    scaleSpace=BCS_BoneSpace
    translationMode=BMM_Additive
    translationSpace=BCS_BoneSpace
    alphaBoolBlend {blendOption=Linear}
    boneToModify {boneName=dyn_shoulderPlate_r}
    rotation=<rot 15 -15 -5>
    scale {x=0.8, z=0.9}
    translation {x=-1}
   }
   n109 = "Transform (Modify) Bone - Bone: dyn_shoulderPlate_r"(
    ComponentPose=@n108
    Alpha=Get Shoulder_rot_back_r_0() -> (Shoulder_rot_back_r_0:double)
    Translation="2,0,0"
    Rotation="0,0,-20"
    Scale="1,1,1"
   ) {
    rotationMode=BMM_Additive
    rotationSpace=BCS_BoneSpace
    scaleMode=BMM_Additive
    scaleSpace=BCS_BoneSpace
    translationMode=BMM_Additive
    translationSpace=BCS_BoneSpace
    alphaBoolBlend {blendOption=Linear}
    boneToModify {boneName=dyn_shoulderPlate_r}
    rotation {roll=-20}
    translation {x=2}
   }
   n110 = "Transform (Modify) Bone - Bone: dyn_arm_l"(ComponentPose=@n109, Alpha=@n103, Translation="3,0,0", Rotation="0,0,0", Scale="1,1,1") {
    rotationMode=BMM_Additive
    rotationSpace=BCS_BoneSpace
    scaleMode=BMM_Additive
    scaleSpace=BCS_BoneSpace
    translationMode=BMM_Additive
    translationSpace=BCS_BoneSpace
    alphaBoolBlend {blendOption=Linear}
    boneToModify {boneName=dyn_arm_l}
    translation {x=3}
   }
   Save cached pose 'TargetingCache'(Pose=Component To Local(ComponentPose=@n110)) {cacheName=TargetingCache}
   Get Shoulder_rot_f_l()
   Get Shoulder_rot_f_r()
  ]}
 EventGraph {nodes [
   Event BlueprintUpdateAnimation()
   > n1 = Sequence!(execute) -> (then_0, then_1, then_2:Exec, then_3:Exec)
   n2 = Get Upperarm_Mesh() -> (Upperarm_Mesh:Skeletal Mesh Component Object Reference)
   > Set Shoulder_rot_f_r(Shoulder_rot_f_r:double=GetAlphaFromReferenceTransformDelta(SkeletalMeshComponent=@n2, BoneName=upperarm_r, BaseName=clavicle_r, InputRange=(X=0,Y=120), TwistAxis="0,0,1")) -> (Output_Get:double)
   > Set Shoulder_rot_f_l(Shoulder_rot_f_l:double=GetAlphaFromReferenceTransformDelta(SkeletalMeshComponent=@n2, BoneName=upperarm_l, BaseName=clavicle_l, InputRange=(X=0,Y=120), TwistAxis="0,0,1")) -> (Output_Get:double)
   > Set Shoulder_rot_back_r(Shoulder_rot_back_r:double=GetAlphaFromReferenceTransformDelta(SkeletalMeshComponent=@n2, BoneName=upperarm_r, BaseName=clavicle_r, InputRange=(X=0,Y=90), TwistAxis="0,0,1")) -> (Output_Get:double)
   > Set Shoulder_rot_back_l(Shoulder_rot_back_l:double=GetAlphaFromReferenceTransformDelta(SkeletalMeshComponent=@n2, BoneName=upperarm_l, BaseName=clavicle_l, InputRange=(X=0,Y=90), TwistAxis="0,0,1")) -> (Output_Get:double)
   n3 = Get Skel_Mesh()
   > Set Shoulder_rot_up_r(Shoulder_rot_up_r:double=GetAlphaFromReferenceTransformDelta(SkeletalMeshComponent=@n3, BoneName=upperarm_r, BaseName=clavicle_r, DeltaComponentType=EulerY, InputRange=(X=0,Y=40), TwistAxis="0,0,1")) -> (Output_Get:double)
   > Set Shoulder_rot_up_l(Shoulder_rot_up_l:double=GetAlphaFromReferenceTransformDelta(
     SkeletalMeshComponent=@n3
     BoneName=upperarm_l
     BaseName=spine_05
     DeltaComponentType=EulerY
     InputRange=(X=-70,Y=90)
     OutputRange=(X=-1,Y=1)
     TwistAxis="0,0,1"
    )) -> (Output_Get:double)
   n4 = Get Upperarm_Mesh()
   > Set Shoulder_rot_dn_r(Shoulder_rot_dn_r:double=GetAlphaFromReferenceTransformDelta(SkeletalMeshComponent=@n4, BoneName=upperarm_r, BaseName=clavicle_r, DeltaComponentType=EulerY, InputRange=(X=60,Y=-40), TwistAxis="0,0,1")) -> (Output_Get:double)
   > Set Shoulder_rot_back_r_0(Shoulder_rot_back_r_0:double=GetAlphaFromReferenceTransformDelta(SkeletalMeshComponent=@n4, BoneName=upperarm_r, BaseName=clavicle_r, InputRange=(X=0,Y=-20), TwistAxis="0,0,1")) -> (Output_Get:double)
   n5 = Get Skel_Mesh()
   > Set Shoulder_rot_up_r_0(Shoulder_rot_up_r_0:double=GetAlphaFromReferenceTransformDelta(SkeletalMeshComponent=@n5, BoneName=upperarm_r, BaseName=clavicle_r, DeltaComponentType=EulerY, InputRange=(X=0,Y=40), TwistAxis="0,0,1")) -> (Output_Get:double)
   > Set Shoulder_rot_up_r_1(Shoulder_rot_up_r_1:double=GetAlphaFromReferenceTransformDelta(SkeletalMeshComponent=@n5, BoneName=clavicle_r, BaseName=spine_05, DeltaComponentType=EulerY, InputRange=(X=0,Y=40), TwistAxis="0,0,1")) -> (Output_Get:double)
   > Set Shoulder_rot_back_l_0(Shoulder_rot_back_l_0:double=GetAlphaFromReferenceTransformDelta(SkeletalMeshComponent=@n4, BoneName=upperarm_l, BaseName=clavicle_l, InputRange=(X=0,Y=-20), TwistAxis="0,0,1")) -> (Output_Get:double)
   > Set Shoulder_rot_dn_l(Shoulder_rot_dn_l:double=GetAlphaFromReferenceTransformDelta(SkeletalMeshComponent=@n4, BoneName=upperarm_l, BaseName=clavicle_l, DeltaComponentType=EulerY, InputRange=(X=60,Y=-40), TwistAxis="0,0,1")) -> (Output_Get:double)
   n6 = Get Skel_Mesh()
   > Set thigh_shape_fn_l(thigh_shape_fn_l:double=GetAlphaFromReferenceTransformDelta(SkeletalMeshComponent=@n6, BoneName=thigh_l, BaseName=pelvis, InputRange=(X=15,Y=-110), TwistAxis="0,0,1")) -> (Output_Get:double)
   > Set thigh_shape_fn_r(thigh_shape_fn_r:double=GetAlphaFromReferenceTransformDelta(SkeletalMeshComponent=@n6, BoneName=thigh_r, BaseName=pelvis, InputRange=(X=15,Y=-110), TwistAxis="0,0,1")) -> (Output_Get:double)
   n7 = Get Skel_Mesh()
   > Set thigh_shape_bk_l(thigh_shape_bk_l:double=GetAlphaFromReferenceTransformDelta(SkeletalMeshComponent=@n7, BoneName=thigh_l, BaseName=pelvis, InputRange=(X=0,Y=-110), TwistAxis="0,0,1")) -> (Output_Get:double)
   > Set thigh_shape_bk_r(thigh_shape_bk_r:double=GetAlphaFromReferenceTransformDelta(SkeletalMeshComponent=@n7, BoneName=thigh_r, BaseName=pelvis, InputRange=(X=0,Y=-110), TwistAxis="0,0,1")) -> (Output_Get:double)
   n8 = Get Skel_Mesh()
   > Set thigh_shape_lt_l(thigh_shape_lt_l:double=GetAlphaFromReferenceTransformDelta(SkeletalMeshComponent=@n8, BoneName=thigh_l, BaseName=pelvis, InputRange=(X=23,Y=-90), TwistAxis="0,0,1")) -> (Output_Get:double)
   > Set thigh_shape_rt_r(thigh_shape_rt_r:double=GetAlphaFromReferenceTransformDelta(SkeletalMeshComponent=@n8, BoneName=thigh_twist_01_r, BaseName=pelvis, InputRange=(X=60,Y=-90), TwistAxis="0,0,1")) -> (Output_Get:double)
   n9 = Get Skel_Mesh()
   > Set thigh_shape_bk_l_0(thigh_shape_bk_l_0:double=GetAlphaFromReferenceTransformDelta(SkeletalMeshComponent=@n9, BoneName=thigh_l, BaseName=pelvis, InputRange=(X=0,Y=60), TwistAxis="0,0,1")) -> (Output_Get:double)
   > Set thigh_shape_bk_r_0(thigh_shape_bk_r_0:double=GetAlphaFromReferenceTransformDelta(SkeletalMeshComponent=@n9, BoneName=thigh_r, BaseName=pelvis, InputRange=(X=0,Y=60), TwistAxis="0,0,1")) -> (Output_Get:double)
   > n10 = Sequence()
   > Set lowerarm_l_front(lowerarm_l_front:double=GetAlphaFromReferenceTransformDelta(SkeletalMeshComponent=Get Skel_Mesh(), BoneName=upperarm_l, BaseName=lowerarm_l, InputRange=(X=0,Y=-70), TwistAxis="0,0,1")) -> (Output_Get:double)
   > Set calf_l_front(calf_l_front:double=GetAlphaFromReferenceTransformDelta(SkeletalMeshComponent=Get Skel_Mesh(), BoneName=thigh_l, BaseName=calf_l, InputRange=(X=0,Y=-90), TwistAxis="0,0,1")) -> (Output_Get:double)
   Set lowerarm_r_front(
    execute=@n10.then_1
    lowerarm_r_front:double=GetAlphaFromReferenceTransformDelta(SkeletalMeshComponent=Get Skel_Mesh(), BoneName=upperarm_r, BaseName=lowerarm_r, InputRange=(X=0,Y=-70), TwistAxis="0,0,1")
   ) -> (Output_Get:double)
   > Set calf_r_front(calf_r_front:double=GetAlphaFromReferenceTransformDelta(SkeletalMeshComponent=Get Skel_Mesh(), BoneName=thigh_r, BaseName=calf_r, InputRange=(X=0,Y=-90), TwistAxis="0,0,1")) -> (Output_Get:double)
   n11 = Cast To FortPlayerAnimInstance(execute=@n1.then_1, Object=GetAnimInstance(self=GetOwnerSkeletalMeshComponent()))
   > Set offset_weapon(offset_weapon:double=0) -> (Output_Get:double)
   > Set offset_weapon_0(offset_weapon_0:double=0) -> (Output_Get:double)
   > Set offset_weapon_1(offset_weapon_1:double=0) -> (Output_Get:double)
   > Set offset_weapon_l_Shoulder(offset_weapon_l_Shoulder:double=0) -> (Output_Get:double)
   > Set offset_weapon_l_Shoulder1(offset_weapon_l_Shoulder1:double=0) -> (Output_Get:double)
   > n12 = Switch on EFortWeaponCoreAnimation!(execute, Selection:EFortWeaponCoreAnimation Enum=Get WeaponCoreAnim(self=@n11)) -> (
    Melee:Exec
    Pistol:Exec
    Shotgun:Exec
    PaperBlueprint:Exec
    Rifle:Exec
    MeleeOneHand:Exec
    MachinePistol:Exec
    RocketLauncher:Exec
    GrenadeLauncher:Exec
    GoingCommando:Exec
    AssaultRifle:Exec
    TacticalShotgun:Exec
    SniperRifle:Exec
    TrapPlacement:Exec
    ShoulderLauncher:Exec
    AbilityDecoTool:Exec
    Crossbow:Exec
    C4:Exec
    RemoteControl:Exec
    DualWield:Exec
    AR_BullPup:Exec
    AR_ForwardGrip:Exec
    MedPackPaddles:Exec
    SMG_P90:Exec
    AR_DrumGun:Exec
    Consumable_Small:Exec
    Consumable_Large:Exec
    Balloon:Exec
    MountedTurret:Exec
    CreativeTool:Exec
    ExplosiveBow:Exec
    AshtonIndigo:Exec
    AshtonChicago:Exec
    MeleeDualWield:Exec
    Unarmed:Exec
   )
   n13 = OR Boolean!(
    A=Get bIsPlayingEmote()
    B=Get bIsSkyDiving()
    C:bool=Get bIsParachuteOpened()
    D:bool=Get bIsSkydiveFloating()
    E:bool=Get bIsSkydiveDiveMode()
    F:bool=Get bIsSurfaceSwimming()
    G:bool=Get bIsTacticalSprinting()
   ) -> (ReturnValue)
   n14 = SelectFloat(A=0, B=Get Shoulder_rot_up_l(), bPickA=@n13)
   Set offset_weapon_l_Shoulder1(execute=[@n12.AssaultRifle, @n12.Shotgun, @n12.TacticalShotgun], offset_weapon_l_Shoulder1=@n14)
   Set offset_weapon(execute=@n12.Pistol, offset_weapon=SelectFloat(A=0, B=Get Shoulder_rot_back_r(), bPickA=@n13))
   > Set offset_weapon_0(offset_weapon_0=SelectFloat(A=1, B=0, bPickA=Get bIsTargeting()))
   > Set offset_weapon_1(offset_weapon_1=SelectFloat(A=0, B=Get Shoulder_rot_up_r(), bPickA=@n13))
   > Set offset_weapon_l_Shoulder(offset_weapon_l_Shoulder=@n14)
   n15 = Branch(execute=@n1.then_2, Condition=Get bIsSkyDiving())
   > Set IsSkyDiving_L(IsSkyDiving_L:double=Get Shoulder_rot_up_l()) -> (Output_Get:double)
   > Set IsSkyDiving_R(IsSkyDiving_R:double=Get Shoulder_rot_up_r_0() -> (Shoulder_rot_up_r_0:double)) -> (Output_Get:double)
   Set IsSkyDiving_L(execute=@n15.else, IsSkyDiving_L=0)
   > Set IsSkyDiving_R(IsSkyDiving_R=0)
   n16 = Branch(execute=@n1.then_3, Condition=Get bIsInFrontEnd())
   > Set isFrontEnd(isFrontEnd:double=1) -> (Output_Get:double)
   Set isFrontEnd(execute=@n16.else, isFrontEnd=0)
   Event BlueprintInitializeAnimation()
   > Set Skel_Mesh(Skel_Mesh=GetOwnerSkeletalMeshComponent())
   > n17 = Cast To PlayerPawn_Athena(Object=TryGetPawnOwner())
   > n18 = Set Pawn Owner(Pawn Owner:Fort Player Pawn Athena Object Reference=@n17) -> (Output_Get:Fort Player Pawn Athena Object Reference)
   > Set Upperarm_Mesh(Upperarm_Mesh:Skeletal Mesh Component Object Reference=GetSkeletalMeshForPartType(self=@n18)) -> (Output_Get:Skeletal Mesh Component Object Reference)
   GetPartSkeletalMeshComponent()
   Get bIsSprinting()
   Get Upperarm_Mesh()
   Get Skel_Mesh()
  ]}
]
variables [
 calf_l_front {type=number}
 calf_r_front {type=number}
 isFrontEnd {type=number}
 IsSkyDiving_L {type=number}
 IsSkyDiving_R {type=number}
 lowerarm_l_back {type=number}
 lowerarm_l_front {type=number}
 lowerarm_r_back {type=number}
 lowerarm_r_front {type=number}
 offset_weapon {type=number}
 offset_weapon_0 {type=number}
 offset_weapon_1 {type=number}
 offset_weapon_l_Shoulder {type=number}
 offset_weapon_l_Shoulder1 {type=number}
 Pawn Owner {type=FortPlayerPawnAthena}
 Shoulder_rot_back_l {type=number}
 Shoulder_rot_back_l_0 {type=number}
 Shoulder_rot_back_r {type=number}
 Shoulder_rot_back_r_0 {type=number}
 Shoulder_rot_dn_l {type=number}
 Shoulder_rot_dn_r {type=number}
 Shoulder_rot_f_l {type=number}
 Shoulder_rot_f_r {type=number}
 Shoulder_rot_up_l {type=number}
 Shoulder_rot_up_r {type=number}
 Shoulder_rot_up_r_0 {type=number}
 Shoulder_rot_up_r_1 {type=number}
 Skel_Mesh {type=SkeletalMeshComponent}
 thigh_shape_bk_l {type=number}
 thigh_shape_bk_l_0 {type=number}
 thigh_shape_bk_r {type=number}
 thigh_shape_bk_r_0 {type=number}
 thigh_shape_fn_l {type=number}
 thigh_shape_fn_r {type=number}
 thigh_shape_lt_l {type=number}
 thigh_shape_rt_r {type=number}
 Upperarm_Mesh {type=SkeletalMeshComponent}
]
