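# Animation Blueprint for the face accessory part of the large male MercurialStorm player character body.
# The AnimGraph copies the pose and curves from the attached parent mesh straight to the output, so the accessory follows that mesh.
# The EventGraph holds an unwired debug Print String that reads the Visor_Dissolve_Progress curve, beside an empty update event.
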AnimBlueprint /BRCosmetics/Characters/Player/Male/Large/Bodies/M_LRG_MercurialStorm/Meshes/Parts/M_LRG_MercurialStorm_FaceAcc_AnimBP
target_skeleton=&/M_LRG_MercurialStorm_FaceAcc_Skeleton
graphs [
 AnimGraph {nodes [Output Pose(Result=Copy Pose From Mesh() {bCopyCurves=true, bUseAttachedParent=true}) {layerGroup=None}]}
 EventGraph {nodes [
   PrintString!(
    execute
    InString="To String (Float)"(InDouble=GetCurveValue(CurveName=Visor_Dissolve_Progress))
    bPrintToScreen
    bPrintToLog
    TextColor
    Duration
    Key
   ) -> (then)
   Event BlueprintUpdateAnimation()
  ]}
]
variables []
