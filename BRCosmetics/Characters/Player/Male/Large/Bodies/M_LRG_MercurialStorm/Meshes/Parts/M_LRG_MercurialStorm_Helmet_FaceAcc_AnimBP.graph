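# Anim Blueprint for the MercurialStorm large male helmet face accessory that corrects neck_01 as the head moves and during skydiving.
# It copies the attached parent mesh pose, then runs two stages of a spine_05/head rotation constraint (weighted toward the spine) plus an additive yaw offset on neck_01.
# The first stage's alpha comes from the head's rotation delta against neck_02.
# The second, with a smaller yaw, switches on with bIsSkyDiving.
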AnimBlueprint /BRCosmetics/Characters/Player/Male/Large/Bodies/M_LRG_MercurialStorm/Meshes/Parts/M_LRG_MercurialStorm_Helmet_FaceAcc_AnimBP
target_skeleton=&/M_LRG_MercurialStorm_Helmet_FaceAcc_Skeleton
graphs [
 AnimGraph {nodes [
   n1 = Get neck_rot_up() -> (neck_rot_up:double)
   n2 = "Constraint - neck_01 (RR)"!(
    ComponentPose=Local To Component(LocalPose=Copy Pose From Mesh() {bUseAttachedParent=true})
    Alpha=@n1
    ConstraintWeights_0=0.7
    ConstraintWeights_1:float=0.3
   ) -> (Pose) {
    alphaBoolBlend {blendOption=Linear}
    boneToModify {boneName=neck_01}
    constraintSetup [{transformType=Rotation, targetBone {boneName=spine_05}}, {transformType=Rotation, targetBone {boneName=head}}]
    constraintWeights [1,1]
   }
   n3 = "Transform (Modify) Bone - Bone: neck_01"(ComponentPose=@n2, Alpha=@n1, Translation="0,0,0", Rotation="0,35,0", Scale="1,1,1") {
    rotationMode=BMM_Additive
    rotationSpace=BCS_BoneSpace
    scaleMode=BMM_Additive
    scaleSpace=BCS_BoneSpace
    translationMode=BMM_Additive
    translationSpace=BCS_BoneSpace
    alphaBoolBlend {blendOption=Linear}
    boneToModify {boneName=neck_01}
    rotation {yaw=35}
   }
   n4 = Get IsSkyDiving() -> (IsSkyDiving:double)
   n5 = "Constraint - neck_01 (RR)"(ComponentPose=@n3, Alpha=@n4, ConstraintWeights_0=0.7, ConstraintWeights_1=0.3) {
    alphaBoolBlend {blendOption=Linear}
    boneToModify {boneName=neck_01}
    constraintSetup [{transformType=Rotation, targetBone {boneName=spine_05}}, {transformType=Rotation, targetBone {boneName=head}}]
    constraintWeights [1,1]
   }
   n6 = "Transform (Modify) Bone - Bone: neck_01"(ComponentPose=@n5, Alpha=@n4, Translation="0,0,0", Rotation="0,17,0", Scale="1,1,1") {
    rotationMode=BMM_Additive
    rotationSpace=BCS_BoneSpace
    scaleMode=BMM_Additive
    scaleSpace=BCS_BoneSpace
    translationMode=BMM_Additive
    translationSpace=BCS_BoneSpace
    alphaBoolBlend {blendOption=Linear}
    boneToModify {boneName=neck_01}
    rotation {yaw=17}
   }
   Output Pose(Result=Component To Local(ComponentPose=@n6)) {layerGroup=None}
  ]}
 EventGraph {nodes [
   Event BlueprintUpdateAnimation()
   > Set neck_rot_up(neck_rot_up:double=GetAlphaFromReferenceTransformDelta(SkeletalMeshComponent=Get Skel_Mesh(), BoneName=head, BaseName=neck_02, InputRange=(X=25,Y=-40), TwistAxis="0,0,1")) -> (Output_Get:double)
   > n1 = Branch(Condition=Get bIsSkyDiving())
   > Set IsSkyDiving(IsSkyDiving:double=1) -> (Output_Get:double)
   Set IsSkyDiving(execute=@n1.else, IsSkyDiving=0)
   Event BlueprintInitializeAnimation()
   > Set Skel_Mesh(Skel_Mesh=GetOwnerSkeletalMeshComponent())
  ]}
]
variables [
 IsSkyDiving {type=number}
 neck_rot_up {type=number}
 Skel_Mesh {type=SkeletalMeshComponent}
]
